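# Default Lit material for a 4K rusty sheet metal albedo texture in the LackLuster look-dev folder.
# Its graph holds a single texture sample of vbuifjj_4K_Albedo, with no connection to any material output shown.
# The sampler reads the albedo as linear color rather than standard color.
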Material /Game/NeverCook/TechArt/LackLuster/LookDev/Textures/RustySheetMetal/vbuifjj_4K_Albedo_Mat
graph {
 properties {pixelDepthOffsetMode=PDOM_Legacy, usedShadingModels=Default Lit}
 nodes [TextureSample() {samplerType=SAMPLERTYPE_LinearColor, texture=&/vbuifjj_4K_Albedo}]
}
